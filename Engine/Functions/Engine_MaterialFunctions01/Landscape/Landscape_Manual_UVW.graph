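# Material function for the library's Landscape category that computes landscape UVW coordinates from manually defined Min and Max positions.
# It subtracts Min Pos from the Position input, divides by the Max minus Min extent, and clamps the normalized result as the UVW output.
# The Position input has world position wired as its preview value.
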MaterialFunction /Engine/Functions/Engine_MaterialFunctions01/Landscape/Landscape_Manual_UVW
graph {
 properties {bExposeToLibrary=true, description=Landscape Manual UVW Defined by Min and Max positions, libraryCategoriesText [Landscape]}
 nodes [
  n1 = FunctionInput() {inputName=Min Pos}
  n2 = Subtract(A=FunctionInput(Preview=WorldPosition().XYZ) {inputName=Position, sortPriority=3}, B=@n1)
  n3 = Subtract(A=FunctionInput() {inputName=Max Pos, sortPriority=1}, B=@n1)
  FunctionOutput(None=Clamp(None=Divide(A=@n2, B=@n3))) {outputName=UVW}
 ]
}
